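; Gameplay ability for WildEstate low gravity that plays a configurable gameplay cue on the player when a valid jump is triggered.
; On event activation it calls the parent, caches the avatar as an Athena player pawn, and runs CheckIfValidJump.
; That check rejects air or ground swinging, falling, crouching, or a failed Fort movement component cast.
; A passing check executes GCToActivate on the owner, with the pawn as instigator and causer.
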(event Ability|EventActivateAbility)

(event Ability|EventOnEndAbility
  (params (bWasCancelled :type Boolean)))

(event Ability|EventActivateAbilityFromEvent
  (params (EventData :type Gameplay%20Event%20Data%20Structure%20%28by%20ref%29))
  (|CallParentFunction EventData)
  (bind _asplayer_pawn_athena (Utilities|Casting|CastToPlayerPawn_Athena (Ability|GetAvatarActorfromActorInfo)))
  (Variables|Default|SetPlayerPawn _asplayer_pawn_athena)
  (if (CallFunction|CheckifValidJump)
    (Ability|ExecuteGameplayCueWithParamsOnOwner (Variables|Default|GetGCToActivate) (Ability|GameplayCue|MakeGameplayCueParameters :NormalizedMagnitude 0.0 :RawMagnitude 0.0 :EffectContext "()" :AggregatedSourceTags "(GameplayTags=)" :AggregatedTargetTags "(GameplayTags=)" :Location "0,0,0" :Normal "0,0,0" :Instigator (Variables|Default|GetPlayerPawn) :EffectCauser (Variables|Default|GetPlayerPawn) :SourceObject self :GameplayEffectLevel 1 :AbilityLevel 1 :bReplicateLocationWhenUsingMinimalRepProxy false))))

(fn CheckIfValidJump
  (bind _playerpawn (Variables|Default|GetPlayerPawn))
  (bind _charactermovement (Variables|/Script/Engine.Character|GetCharacterMovement _playerpawn))
  (bind _asfort_movement_comp_character (Utilities|Casting|CastToFortMovementComp_Character _charactermovement)
    (:then
      (return (not (Math|Boolean|ORBoolean (Swinging|IsSwingingAir _asfort_movement_comp_character) (Swinging|IsSwingingGround _asfort_movement_comp_character) (AI|Components|NavMovement|IsFalling _charactermovement) (AI|Components|NavMovement|IsCrouching _charactermovement)))))
    (:CastFailed
      (return false))))
